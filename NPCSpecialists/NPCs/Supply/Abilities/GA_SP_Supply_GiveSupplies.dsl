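; Gameplay ability where a Supply specialist NPC tosses ammo, materials or utility items as gift pickups.
; Weighted rolls choose the category, optionally using a loot tier table for non-ammo drops.
; Ammo and materials follow the target player's weapon and inventory, with fallbacks for full stock and no-build playlists, including bandages.
; Each drop triggers a montage event, spawns tagged pickups mid-animation, optionally marks them, then lingers before ending and committing cooldown.
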(event Ability|EventActivateAbility
  (bind _returnvalue (Ability|GetAvatarActorfromActorInfo))
  ;[1 1 1] Initializing variables and setting up drop chance for materials or ammo
  (Ability|CommitAbilityCost false)
  (bind _asfort_player_pawn (Utilities|Casting|CastToFortPlayerPawn _returnvalue)
    (:then
      (Variables|InternalData|SetBotPawn _asfort_player_pawn)
      (bind _asfort_athena_aibot_controller (Utilities|Casting|CastToFortAthenaAIBotController (AI|GetAIController _returnvalue))
        (:then
          (Variables|InternalData|SetBotController _asfort_athena_aibot_controller)
          (bind _asfort_player_controller_athena (Utilities|Casting|CastToFortPlayerControllerAthena (CallFunction|GetConverterPlayerController))
            (:then
              (Variables|Default|SetMarkerComponent (Class|FortPlayerController|GetMarkerComponent _asfort_player_controller_athena))
              (CallFunction|StartDrop))
            (:CastFailed
              (Ability|EndAbility))))
        (:CastFailed
          (Ability|EndAbility))))
    (:CastFailed
      (Ability|EndAbility))))

(event Custom|SpawnDropsFromLootTier
  (bind (_position _direction) (CallFunction|SpawnPickupPositionandDIrection))
  (bind (_gameplayeventdata _getplaylength) (CallFunction|AnimPayload))
  ;[1 1 1] Simplify Logic to Spawn Mats and Utilities by using the LootTier System that does all of that for us - also easier to support by Balance Team
  (Ability|SendGameplayEventtoActor (Variables|InternalData|GetBotPawn) (Variables|InternalData|GetAnimTag) _gameplayeventdata)
  (Utilities|FlowControl|Delay (/ _getplaylength (ScalableFloat|GetValueatLevel (Variables|Pickups|GetMontagePlayLengthDivisionToSpawnPickup))))
  (bind _itempickupsspawned (Class|AthenaFunctionLibrary|SpawnLootfromTable (Variables|Pickups|LootTier|GetMatsAndUtilityLootTier) 1 -1 _position _direction 0 (Variables|Pickups|LootTier|GetTossWhenLootIsGivenByLootTier) (Variables|Pickups|LootTier|GetRandomRotationWhenLootIsGivenByLootTier) false -1 "Tossed" "NPCGift"))
  (CallFunction|AddSourceTagsforEachItemSpawned (GameplayTags|GetOwnedGameplayTags (Variables|InternalData|GetBotPawn) :!class /Script/GameplayTags.GameplayTagAssetInterface) _itempickupsspawned)
  (if (and (> (Utilities|Array|Length _itempickupsspawned) 0) (== (ScalableFloat|GetValueatLevel (Variables|Pickups|GetEnablePickupMarking)) (Variables|InternalData|GetDefaultPickupMarkingValueForActivationCheck)))
    (Utilities|FlowControl|Delay (ScalableFloat|GetValueatLevel (Variables|Pickups|GetDelayBeforeMarkingItem))) ; Waiting for pickup to be in place before marking
    (MarkerComponent|MarkActoronServer (Variables|Default|GetMarkerComponent) (Utilities|Array|Get(acopy) _itempickupsspawned) true true))
  (CallFunction|EndGiveSupplyAbility))

(event Custom|StartDrop
  ;[1 1 1] Choose What To Drop
  (Ability|SendGameplayEventtoActor (Variables|InternalData|GetBotPawn) "(TagName=\"Event.Ability.NPC.SpecialistAction\")")
  (if (Math|Random|RandomBoolwithWeight (ScalableFloat|GetValueatLevel (Variables|Pickups|GetNonAmmoDropChance)))
    (if (Variables|Pickups|LootTier|GetUseLootTiersToSpawnMatAndUtility)
      (CallFunction|SpawnDropsfromLootTier)
      (elif (Math|Random|RandomBoolwithWeight (ScalableFloat|GetValueatLevel (Variables|Pickups|GetUtilityDropChance)))
        (CallFunction|DropUtility)
        (else
          (CallFunction|DropMats))))
    (else
      (CallFunction|DropAmmo))))

(event Custom|GiveBandage
  (bind (_position _direction) (CallFunction|SpawnPickupPositionandDIrection))
  (bind (_gameplayeventdata _getplaylength) (CallFunction|AnimPayload))
  ;[1 1 1] Drop ammo logic
  (Ability|SendGameplayEventtoActor (Variables|InternalData|GetBotPawn) (Variables|InternalData|GetAnimTag) _gameplayeventdata)
  (Utilities|FlowControl|Delay (/ _getplaylength (ScalableFloat|GetValueatLevel (Variables|Pickups|GetMontagePlayLengthDivisionToSpawnPickup))))
  (bind _returnvalue (Inventory|SpawnPickupInWorld (Variables|InternalData|GetSelectedAmmo) (Variables|InternalData|GetNumberOfAmmo) _position _direction 0 false false false -1 "Tossed" "NPCGift" 0 false))
  (Class|FortPickup|AddSpawnSourceTags _returnvalue (GameplayTags|GetOwnedGameplayTags (Variables|InternalData|GetBotPawn) :!class /Script/GameplayTags.GameplayTagAssetInterface))
  (if (== (ScalableFloat|GetValueatLevel (Variables|Pickups|GetEnablePickupMarking)) (Variables|InternalData|GetDefaultPickupMarkingValueForActivationCheck))
    (Utilities|FlowControl|Delay (ScalableFloat|GetValueatLevel (Variables|Pickups|GetDelayBeforeMarkingItem))) ; Waiting for pickup to be in place before marking
    (MarkerComponent|MarkActoronServer (Variables|Default|GetMarkerComponent) _returnvalue true true))
  (CallFunction|EndGiveSupplyAbility))

(event Custom|DropAmmo
  (bind _utilitydropchance (Variables|Pickups|GetUtilityDropChance))
  (bind _returnvalue (ScalableFloat|GetValueatLevel _utilitydropchance))
  (bind _returnvalue_2 (Math|Random|RandomBoolwithWeight _returnvalue))
  (bind (_position _direction) (CallFunction|SpawnPickupPositionandDIrection))
  (bind (_gameplayeventdata _getplaylength) (CallFunction|AnimPayload))
  ;[1 1 1] Drop ammo logic
  (bind _ammoisfull (CallFunction|AmmoDropManager))
  (if _ammoisfull
    ;[1 1 1] Did not give ammo, because current equipped weapon ammo is full and fallback to give random ammo is also full. So give something else instead.
    (if (Variables|Pickups|LootTier|GetUseLootTiersToSpawnMatAndUtility)
      (CallFunction|SpawnDropsfromLootTier)
      (else
        (bind _is_no_build_playlist (CallFunction|NoBuildPlaylistCheck))
        (if _is_no_build_playlist
          (if _returnvalue_2
            (CallFunction|DropUtility)
            (else
              (CallFunction|SetupBandagetoGive)
              (CallFunction|GiveBandage)))
          (elif _returnvalue_2
            (CallFunction|DropUtility)
            (else
              (CallFunction|DropMats))))))
    (else
      (Ability|SendGameplayEventtoActor (Variables|InternalData|GetBotPawn) (Variables|InternalData|GetAnimTag) _gameplayeventdata)
      (Utilities|FlowControl|Delay (/ _getplaylength (ScalableFloat|GetValueatLevel (Variables|Pickups|GetMontagePlayLengthDivisionToSpawnPickup))))
      (bind _returnvalue_3 (Inventory|SpawnPickupInWorld (Variables|InternalData|GetSelectedAmmo) (Variables|InternalData|GetNumberOfAmmo) _position _direction 0 false false false -1 "Tossed" "NPCGift" 0 false))
      (Class|FortPickup|AddSpawnSourceTags _returnvalue_3 (GameplayTags|GetOwnedGameplayTags (Variables|InternalData|GetBotPawn) :!class /Script/GameplayTags.GameplayTagAssetInterface))
      (if (== (ScalableFloat|GetValueatLevel (Variables|Pickups|GetEnablePickupMarking)) (Variables|InternalData|GetDefaultPickupMarkingValueForActivationCheck))
        (Utilities|FlowControl|Delay (ScalableFloat|GetValueatLevel (Variables|Pickups|GetDelayBeforeMarkingItem))) ; Waiting for pickup to be in place before marking
        (MarkerComponent|MarkActoronServer (Variables|Default|GetMarkerComponent) _returnvalue_3 true true))
      (CallFunction|EndGiveSupplyAbility))))

(event Custom|NoBuildPlaylistFallback
  (bind _utilitydropchance (Variables|Pickups|GetUtilityDropChance))
  (bind _returnvalue (ScalableFloat|GetValueatLevel _utilitydropchance))
  (bind _returnvalue_2 (Math|Random|RandomBoolwithWeight _returnvalue))
  (bind (_position _direction) (CallFunction|SpawnPickupPositionandDIrection))
  (bind (_gameplayeventdata _getplaylength) (CallFunction|AnimPayload))
  ;[1 1 1] Drop ammo logic
  (bind _ammoisfull (CallFunction|AmmoDropManager))
  (if _ammoisfull
    ;[1 1 1] Did not give ammo, because current equipped weapon ammo is full and fallback to give random ammo is also full. So give something else instead.
    (if (Variables|Pickups|LootTier|GetUseLootTiersToSpawnMatAndUtility)
      (CallFunction|SpawnDropsfromLootTier)
      (else
        (bind _is_no_build_playlist (CallFunction|NoBuildPlaylistCheck))
        (if _is_no_build_playlist
          (if _returnvalue_2
            (CallFunction|DropUtility)
            (else
              (CallFunction|SetupBandagetoGive)
              (CallFunction|GiveBandage)))
          (elif _returnvalue_2
            (CallFunction|DropUtility)
            (else
              (CallFunction|DropMats))))))
    (else
      (Ability|SendGameplayEventtoActor (Variables|InternalData|GetBotPawn) (Variables|InternalData|GetAnimTag) _gameplayeventdata)
      (Utilities|FlowControl|Delay (/ _getplaylength (ScalableFloat|GetValueatLevel (Variables|Pickups|GetMontagePlayLengthDivisionToSpawnPickup))))
      (bind _returnvalue_3 (Inventory|SpawnPickupInWorld (Variables|InternalData|GetSelectedAmmo) (Variables|InternalData|GetNumberOfAmmo) _position _direction 0 false false false -1 "Tossed" "NPCGift" 0 false))
      (Class|FortPickup|AddSpawnSourceTags _returnvalue_3 (GameplayTags|GetOwnedGameplayTags (Variables|InternalData|GetBotPawn) :!class /Script/GameplayTags.GameplayTagAssetInterface))
      (if (== (ScalableFloat|GetValueatLevel (Variables|Pickups|GetEnablePickupMarking)) (Variables|InternalData|GetDefaultPickupMarkingValueForActivationCheck))
        (Utilities|FlowControl|Delay (ScalableFloat|GetValueatLevel (Variables|Pickups|GetDelayBeforeMarkingItem))) ; Waiting for pickup to be in place before marking
        (MarkerComponent|MarkActoronServer (Variables|Default|GetMarkerComponent) _returnvalue_3 true true))
      (CallFunction|EndGiveSupplyAbility))))

(event Custom|DropMats
  (bind (_position _direction) (CallFunction|SpawnPickupPositionandDIrection))
  (bind (_gameplayeventdata _getplaylength) (CallFunction|AnimPayload))
  ;[1 1 1] Material drop logic
  (bind _is_no_build_playlist (CallFunction|NoBuildPlaylistCheck))
  (if _is_no_build_playlist
    (CallFunction|NoBuildPlaylistFallback)
    (else
      (bind (_wasmaterialfoundtogive _materialtogive) (Default|MaterialDropManager))
      (if _wasmaterialfoundtogive
        (Variables|InternalData|SetSelectedMaterial _materialtogive)
        (CallFunction|ConfiguringMaterialQuantity)
        (Ability|SendGameplayEventtoActor (Variables|InternalData|GetBotPawn) (Variables|InternalData|GetAnimTag) _gameplayeventdata)
        (Utilities|FlowControl|Delay (/ _getplaylength (ScalableFloat|GetValueatLevel (Variables|Pickups|GetMontagePlayLengthDivisionToSpawnPickup))))
        (bind _returnvalue (Inventory|SpawnPickupInWorld (Variables|InternalData|GetSelectedMaterial) (Variables|InternalData|GetNumberOfMaterials) _position _direction 0 false false false -1 "Tossed" "NPCGift" 0 false))
        (Class|FortPickup|AddSpawnSourceTags _returnvalue (GameplayTags|GetOwnedGameplayTags (Variables|InternalData|GetBotPawn) :!class /Script/GameplayTags.GameplayTagAssetInterface))
        (if (== (ScalableFloat|GetValueatLevel (Variables|Pickups|GetEnablePickupMarking)) (Variables|InternalData|GetDefaultPickupMarkingValueForActivationCheck))
          (Utilities|FlowControl|Delay (ScalableFloat|GetValueatLevel (Variables|Pickups|GetDelayBeforeMarkingItem))) ; Waiting for pickup to be in place before marking
          (MarkerComponent|MarkActoronServer (Variables|Default|GetMarkerComponent) _returnvalue true true))
        (CallFunction|EndGiveSupplyAbility)
        (else
          (CallFunction|SetupBandagetoGive)
          (CallFunction|GiveBandage))))))

(event Custom|GiveMats
  (bind (_position _direction) (CallFunction|SpawnPickupPositionandDIrection))
  (bind (_gameplayeventdata _getplaylength) (CallFunction|AnimPayload))
  ;[1 1 1] Material drop logic
  (bind (_wasmaterialfoundtogive _materialtogive) (Default|MaterialDropManager))
  (if _wasmaterialfoundtogive
    (Variables|InternalData|SetSelectedMaterial _materialtogive)
    (CallFunction|ConfiguringMaterialQuantity)
    (Ability|SendGameplayEventtoActor (Variables|InternalData|GetBotPawn) (Variables|InternalData|GetAnimTag) _gameplayeventdata)
    (Utilities|FlowControl|Delay (/ _getplaylength (ScalableFloat|GetValueatLevel (Variables|Pickups|GetMontagePlayLengthDivisionToSpawnPickup))))
    (bind _returnvalue (Inventory|SpawnPickupInWorld (Variables|InternalData|GetSelectedMaterial) (Variables|InternalData|GetNumberOfMaterials) _position _direction 0 false false false -1 "Tossed" "NPCGift" 0 false))
    (Class|FortPickup|AddSpawnSourceTags _returnvalue (GameplayTags|GetOwnedGameplayTags (Variables|InternalData|GetBotPawn) :!class /Script/GameplayTags.GameplayTagAssetInterface))
    (if (== (ScalableFloat|GetValueatLevel (Variables|Pickups|GetEnablePickupMarking)) (Variables|InternalData|GetDefaultPickupMarkingValueForActivationCheck))
      (Utilities|FlowControl|Delay (ScalableFloat|GetValueatLevel (Variables|Pickups|GetDelayBeforeMarkingItem))) ; Waiting for pickup to be in place before marking
      (MarkerComponent|MarkActoronServer (Variables|Default|GetMarkerComponent) _returnvalue true true))
    (CallFunction|EndGiveSupplyAbility)
    (else
      (CallFunction|SetupBandagetoGive)
      (CallFunction|GiveBandage))))

(event Custom|DropUtility
  (bind (_gameplayeventdata _getplaylength) (CallFunction|AnimPayload))
  (bind (_position _direction) (CallFunction|SpawnPickupPositionandDIrection))
  ;[1 1 1] Utility drop Logic
  (Default|UtilityDropManager)
  (Ability|SendGameplayEventtoActor (Variables|InternalData|GetBotPawn) (Variables|InternalData|GetAnimTag) _gameplayeventdata)
  (Utilities|FlowControl|Delay (/ _getplaylength (ScalableFloat|GetValueatLevel (Variables|Pickups|GetMontagePlayLengthDivisionToSpawnPickup))))
  (bind _returnvalue (Inventory|SpawnPickupInWorld (Variables|InternalData|GetSelectedUtility) (Variables|InternalData|GetNumberOfUtilities) _position _direction 0 false false false -1 "Tossed" "NPCGift" 0 false))
  (Class|FortPickup|AddSpawnSourceTags _returnvalue (GameplayTags|GetOwnedGameplayTags (Variables|InternalData|GetBotPawn) :!class /Script/GameplayTags.GameplayTagAssetInterface))
  (if (== (ScalableFloat|GetValueatLevel (Variables|Pickups|GetEnablePickupMarking)) (Variables|InternalData|GetDefaultPickupMarkingValueForActivationCheck))
    (Utilities|FlowControl|Delay (ScalableFloat|GetValueatLevel (Variables|Pickups|GetDelayBeforeMarkingItem))) ; Waiting for pickup to be in place before marking
    (MarkerComponent|MarkActoronServer (Variables|Default|GetMarkerComponent) _returnvalue true true))
  (CallFunction|EndGiveSupplyAbility))

(event Custom|EndGiveSupplyAbility
  ;[1 1 1] Ending Give Supply ability with a delay for the NPC to stay looking at the player after the drop to feel less hectic
  (Ability|SendGameplayEventtoActor (Variables|InternalData|GetBotPawn) (Variables|Default|GetTag_SupplyWasGiven))
  (Utilities|FlowControl|Delay (ScalableFloat|GetValueatLevel (Variables|Pickups|GetDelayAfterDropItem)))
  (Ability|EndAbility))

(event Ability|EventOnEndAbility
  (params (bWasCancelled :type Boolean))
  ;[1 1 1] On end ability to commit ability cooldown
  (|CallParentFunction bWasCancelled)
  (Ability|CommitAbilityCooldown false))

(event Ability|EventFailedToActivatePassiveAbility
  (params (ActorInfo :type Gameplay%20Ability%20Actor%20Info%20Structure)))

(fn AddSourceTagsForEachItemSpawned
  (params (AdditionalTagsToAdd :type Gameplay%20Tag%20Container%20Structure%20%28by%20ref%29) (Array :type Array%20of%20Fort%20Pickup%20Object%20References))
  (for _array_element Array
    (Utilities|IsValid _array_element
      (:Is%20Valid
        (Class|FortPickup|AddSpawnSourceTags _array_element AdditionalTagsToAdd)))))

(fn AmmoDropManager
  (bind _targetedpawn (Variables|InternalData|GetTargetedPawn))
  (bind _returnvalue (Weapon|GetCurrentWeapon _targetedpawn))
  (bind _targetpawn (CallFunction|GetTargetPawnfromBlackboard))
  (bind _returnvalue_2 (WeaponItemDefinition|GetAmmoWorldItemDefinition (Data|GetWeaponData _returnvalue)))
  (bind _returnvalue_3 (Data|GetWeaponData _returnvalue))
  (Utilities|IsValid _targetpawn
    (:Is%20Valid
      (bind _output_get (Variables|InternalData|SetTargetedPawn _targetpawn))
      (bind _asfort_player_controller (Utilities|Casting|CastToFortPlayerController (Pawn|GetController _output_get))
        (:then
          (Variables|InternalData|SetTargetedPlayerController _asfort_player_controller)
          (Utilities|IsValid _returnvalue
            (:Is%20Valid
              (Utilities|IsValid _returnvalue_3
                (:Is%20Valid
                  (Utilities|IsValid _returnvalue_2
                    (:Is%20Valid
                      ;[1 1 1] Checking if the player's ammo quantity is maxed out
                      (Variables|InternalData|SetSelectedAmmo _returnvalue_2)
                      (Variables|InternalData|SetPlayerItemQuantity (Inventory|GetItemQuantityOnPlayer (Variables|InternalData|GetTargetedPlayerController) (Variables|InternalData|GetSelectedAmmo)))
                      (Variables|InternalData|SetItemMaxStackSize (Item|GetMaxStackSize (Variables|InternalData|GetSelectedAmmo) 0))
                      (if (>= (Variables|InternalData|GetPlayerItemQuantity) (Variables|InternalData|GetItemMaxStackSize))
                        ;[1 1 1] Detecting ammo type to set quantity. Prioritize via ammo popularity.
                        (bind _isnewammoalreadyfull (CallFunction|NewAmmoFinder))
                        (if _isnewammoalreadyfull
                          (return true)
                          (else
                            _)))
                      (if (or (GameplayTags|HasMatchingGameplayTag (Variables|InternalData|GetSelectedAmmo) "(TagName=\"supplies.ammo.bulletlight\")") (GameplayTags|HasMatchingGameplayTag (Variables|InternalData|GetSelectedAmmo) "(TagName=\"supplies.ammo.bulletmedium\")"))
                        (Default|SetAmmoLightorMediumQuantity)
                        (return false)
                        (elif (GameplayTags|HasMatchingGameplayTag (Variables|InternalData|GetSelectedAmmo) "(TagName=\"supplies.ammo.bulletheavy\")")
                          (Default|SetAmmoHeavyQuantity)
                          (return false)
                          (elif (GameplayTags|HasMatchingGameplayTag (Variables|InternalData|GetSelectedAmmo) "(TagName=\"supplies.ammo.shell\")")
                            (Default|SetAmmoShellQuantity)
                            (return false)
                            (elif (GameplayTags|HasMatchingGameplayTag (Variables|InternalData|GetSelectedAmmo) "(TagName=\"supplies.ammo.explosive\")")
                              (CallFunction|SetAmmoRocketQuantity)
                              (return false)
                              (elif (Math|Boolean|ORBoolean (GameplayTags|HasMatchingGameplayTag (Variables|InternalData|GetSelectedAmmo) "(TagName=\"Item.Consumable.Grenade\")") (GameplayTags|HasMatchingGameplayTag (Variables|InternalData|GetSelectedAmmo) "(TagName=\"Item.Consumable.ShockGrenade\")") (GameplayTags|HasMatchingGameplayTag (Variables|InternalData|GetSelectedAmmo) "(TagName=\"Item.Consumable.PortAShack\")") false)
                                (Default|SetAmmoGrenadetQuantity)
                                (return false)
                                (elif (GameplayTags|HasMatchingGameplayTag (Variables|InternalData|GetSelectedAmmo) "(TagName=\"supplies.ammo.arrows\")")
                                  (Default|SetAmmoBowQuantity)
                                  (return false)
                                  (else
                                    (Default|FallbackAmmo)
                                    (return false)))))))))
                    (:Is%20Not%20Valid
                      _)))
                (:Is%20Not%20Valid
                  _)))
            (:Is%20Not%20Valid
              _)))
        (:CastFailed
          (Ability|EndAbility))))
    (:Is%20Not%20Valid
      _))
  ;[1 1 1] Fallback if weapon or ammo retrieve is invalid
  (Default|FallbackAmmo)
  (return false))

(fn AnimPayload
  (return (Utilities|Struct|MakeGameplayEventData "(TagName=\"\")" 0 0 (Utilities|Struct|BreakFortGameplayAbilityMontageInfo (Variables|InternalData|GetMontageInfoForScan))) (Animation|GetPlayLength (Utilities|Struct|BreakFortGameplayAbilityMontageInfo (Variables|InternalData|GetMontageInfoForScan)))))

(fn ConfiguringMaterialQuantity
  (bind _selectedmaterial (Variables|InternalData|GetSelectedMaterial))
  (if (== _selectedmaterial "/Game/Items/ResourcePickups/WoodItemData.WoodItemData")
    (Variables|InternalData|SetNumberOfMaterials (ScalableFloat|GetValueasInteger (Variables|Pickups|GetMaterialWoodQuantity)))
    (elif (== _selectedmaterial "/Game/Items/ResourcePickups/StoneItemData.StoneItemData")
      (Variables|InternalData|SetNumberOfMaterials (ScalableFloat|GetValueasInteger (Variables|Pickups|GetMaterialStoneQuantity)))
      (elif (== _selectedmaterial "/Game/Items/ResourcePickups/MetalItemData.MetalItemData")
        (Variables|InternalData|SetNumberOfMaterials (ScalableFloat|GetValueasInteger (Variables|Pickups|GetMaterialMetalQuantity)) :!label _label)
        (else
          (goto _label)))))
  (return))

(fn FallbackAmmo
  ;[1 1 1] Fallback if initialization fails
  (switch int (Math|Random|RandomInteger 2)
    (:0
      (Variables|InternalData|SetSelectedAmmo (Variables|Pickups|GetAmmoTypeBulletsLight))
      (Variables|InternalData|SetNumberOfAmmo (ScalableFloat|GetValueasInteger (Variables|Pickups|GetAmmoDefaultQuantity))))
    (:1
      (Variables|InternalData|SetSelectedAmmo (Variables|Pickups|GetAmmoTypeBulletsMedium))
      (Variables|InternalData|SetNumberOfAmmo (ScalableFloat|GetValueasInteger (Variables|Pickups|GetAmmoDefaultQuantity)))))
  (return))

(fn GetConverterPlayerController
  (return (Fortnite|GetFortPlayerControllerfromActor (Convert|GetConverterPawn (Variables|/Script/FortniteGame.FortPlayerPawn|GetConvertComponent (Variables|InternalData|GetBotPawn))))))

(fn GetTargetPawnFromBlackboard
  (bind _returnvalue (CallFunction|GetAvatarasFortPlayerPawn))
  (bind _returnvalue_2 (AI|GetAIController _returnvalue))
  (bind _returnvalue_3 (AI|GetBlackboard _returnvalue_2))
  (Utilities|IsValid _returnvalue_2
    (:Is%20Valid
      (Utilities|IsValid _returnvalue_3
        (:Is%20Valid
          (bind _asfort_pawn (Utilities|Casting|CastToFortPawn (AI|Components|Blackboard|GetValueasObject _returnvalue_3 (LocalVariables|GetBBKey_TargetActor)))
            (:then
              (return _asfort_pawn))
            (:CastFailed
              (return 0))))))))

(fn IsNewAmmoGoodToGive
  (params (ItemDefinition :type Fort%20Item%20Definition%20Object%20Reference))
  (LocalVariables|SetNewPlayerItemQuantity (Inventory|GetItemQuantityOnPlayer (Variables|InternalData|GetTargetedPlayerController) ItemDefinition))
  (LocalVariables|SetNewItemMaxStackSize (Item|GetMaxStackSize ItemDefinition 0))
  (return (>= (LocalVariables|GetNewPlayerItemQuantity) (LocalVariables|GetNewItemMaxStackSize))))

(fn MaterialDropManager
  (bind _targetpawn (CallFunction|GetTargetPawnfromBlackboard))
  (Utilities|Array|Clear (Variables|InternalData|GetListOfMaterialsToGive))
  (Utilities|IsValid _targetpawn
    (:Is%20Valid
      (Variables|InternalData|SetTargetedPawn _targetpawn)
      (bind _asfort_player_controller (Utilities|Casting|CastToFortPlayerController (Pawn|GetController (Variables|InternalData|GetTargetedPawn)))
        (:then
          (Variables|InternalData|SetTargetedPlayerController _asfort_player_controller)
          (Utilities|FlowControl|Sequence
            (:then_0
              (bind _returnvalue (Building|K2GetNumAvailableBuildingResources (Variables|InternalData|GetTargetedPlayerController)))
              (if (<= _returnvalue (ScalableFloat|GetValueasInteger (Variables|Pickups|GetMaterialDropThreshold)))
                (Utilities|Array|Add (Variables|InternalData|GetListOfMaterialsToGive) (Variables|InternalData|GetWood))))
            (:then_1
              (bind _returnvalue_2 (Building|K2GetNumAvailableBuildingResources (Variables|InternalData|GetTargetedPlayerController) "Stone"))
              (if (<= _returnvalue_2 (ScalableFloat|GetValueasInteger (Variables|Pickups|GetMaterialDropThreshold)))
                (Utilities|Array|Add (Variables|InternalData|GetListOfMaterialsToGive) (Variables|InternalData|GetStone))))
            (:then_2
              (bind _returnvalue_3 (Building|K2GetNumAvailableBuildingResources (Variables|InternalData|GetTargetedPlayerController) "Metal"))
              (if (<= _returnvalue_3 (ScalableFloat|GetValueasInteger (Variables|Pickups|GetMaterialDropThreshold)))
                (Utilities|Array|Add (Variables|InternalData|GetListOfMaterialsToGive) (Variables|InternalData|GetMetal))))
            (:then_3
              (if (Utilities|Array|IsEmpty (Variables|InternalData|GetListOfMaterialsToGive))
                (return false 0)
                (else
                  (return true (Utilities|Array|RandomArrayItem (Variables|InternalData|GetListOfMaterialsToGive))))))))
        (:CastFailed
          (Ability|EndAbility))))
    (:Is%20Not%20Valid
      (Ability|EndAbility))))

(fn NewAmmoFinder
  (switch int (Math|Random|RandomInteger 4)
    (:0
      (bind _output_get (Variables|InternalData|SetSelectedAmmo (Variables|InternalData|GetAmmoShell)))
      (bind _isnewammoalreadyfull (CallFunction|IsNewAmmoGoodtoGive _output_get))
      (return _isnewammoalreadyfull))
    (:1
      (bind _output_get_2 (Variables|InternalData|SetSelectedAmmo (Variables|InternalData|GetAmmoBulletsLight)))
      (bind _isnewammoalreadyfull_2 (CallFunction|IsNewAmmoGoodtoGive _output_get_2))
      (return _isnewammoalreadyfull_2))
    (:2
      (bind _output_get_3 (Variables|InternalData|SetSelectedAmmo (Variables|InternalData|GetAmmoBulletsMedium)))
      (bind _isnewammoalreadyfull_3 (CallFunction|IsNewAmmoGoodtoGive _output_get_3))
      (return _isnewammoalreadyfull_3))
    (:3
      (bind _output_get_4 (Variables|InternalData|SetSelectedAmmo (Variables|InternalData|GetAmmoBulletsHeavy)))
      (bind _isnewammoalreadyfull_4 (CallFunction|IsNewAmmoGoodtoGive _output_get_4))
      (return _isnewammoalreadyfull_4))
    (:4
      (bind _output_get_5 (Variables|InternalData|SetSelectedAmmo (Variables|InternalData|GetAmmoRockets)))
      (bind _isnewammoalreadyfull_5 (CallFunction|IsNewAmmoGoodtoGive _output_get_5))
      (return _isnewammoalreadyfull_5))))

(fn NoBuildPlaylistCheck
  (return (Class|FortGameStateAthena|IsNoBuildBRPlaylist (Fortnite|Athena|GameState|GetGameStateAthena))))

(fn SetAmmoBowQuantity
  (Variables|InternalData|SetNumberOfAmmo (ScalableFloat|GetValueasInteger (Variables|Pickups|GetAmmoBowQuantity)))
  (return))

(fn SetAmmoGrenadetQuantity
  (Variables|InternalData|SetNumberOfAmmo (ScalableFloat|GetValueasInteger (Variables|Pickups|GetAmmoGrenadeQuantity)))
  (return))

(fn SetAmmoHeavyQuantity
  (Variables|InternalData|SetNumberOfAmmo (ScalableFloat|GetValueasInteger (Variables|Pickups|GetAmmoHeavyQuantity)))
  (return))

(fn SetAmmoLightOrMediumQuantity
  (Variables|InternalData|SetNumberOfAmmo (ScalableFloat|GetValueasInteger (Variables|Pickups|GetAmmoDefaultQuantity)))
  (return))

(fn SetAmmoRocketQuantity
  (Variables|InternalData|SetNumberOfAmmo (ScalableFloat|GetValueasInteger (Variables|Pickups|GetAmmoRocketQuantity)))
  (return))

(fn SetAmmoShellQuantity
  (Variables|InternalData|SetNumberOfAmmo (ScalableFloat|GetValueasInteger (Variables|Pickups|GetAmmoShellQuantity)))
  (return))

(fn SetupBandageToGive
  (Variables|InternalData|SetSelectedAmmo (Variables|Pickups|GetItemBandage))
  (Variables|InternalData|SetNumberOfAmmo (ScalableFloat|GetValueasInteger (Variables|Pickups|GetItemBandageQuantity)))
  (return))

(fn SpawnPickupPositionAndDIrection
  (return (+ (Math|Vector|RotateVectorAroundAxis (Utilities|Operators|Multiply (Transformation|GetActorForwardVector (Variables|InternalData|GetBotPawn)) (Math|Vector|MakeVector (Math|Random|RandomFloatinRange 1.4e+02 1.8e+02) (Math|Random|RandomFloatinRange 1.4e+02 1.8e+02))) (Math|Random|RandomFloatinRange -2e+01 2e+01) "0.000000,0.000000,1.000000") (Transformation|GetActorLocation (Variables|InternalData|GetBotPawn))) (Transformation|GetActorForwardVector (Variables|InternalData|GetBotPawn))))

(fn UtilityDropManager
  (Variables|InternalData|SetSelectedUtility (Utilities|Array|RandomArrayItem (Variables|Pickups|GetListOfUtilities)))
  (if (== (Variables|InternalData|GetUtilityAthenaGrenade) (Variables|InternalData|GetSelectedUtility))
    _
    (elif (== (Variables|InternalData|GetUtilityFireFlies) (Variables|InternalData|GetSelectedUtility))
      _
      (else
        (Variables|InternalData|SetNumberOfUtilities (ScalableFloat|GetValueasInteger (Variables|Pickups|GetUtilityDefaultQuantity)))
        (return))))
  (Variables|InternalData|SetNumberOfUtilities (ScalableFloat|GetValueasInteger (Variables|Pickups|GetUtilityGrenadeQuantity)))
  (return))
